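# Physics asset for the KnightCatRacket backpack cosmetic mesh, giving its dangling zipper simple physical motion and collision against the pack.
# Two capsule bodies are defined, one on the pack bone and one on the din_zipper_base bone.
# A single angularly limited constraint joins the zipper to the pack, allowing wider swing on one axis and tight swing and twist on the others.
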PhysicsAsset /BRCosmetics/Accessories/FORT_Backpacks/Backpack_F_MED_KnightCatRacket/Meshes/F_MED_KnightCatRacket_Pack_Physics
bodies {
 din_zipper_base [{shapeType=Capsule, length=1.25, radius=0.75, shapeName=din_zipper_base_capsule, center=<1.7503 0 0>, rotation=<rot -90 -166.7373 166.7373>}]
 pack [{shapeType=Capsule, length=8.25, radius=2.5, shapeName=pack_capsule, center=<0 -2.2586 0>}]
}
constraints [{bone1Name=din_zipper_base, bone2Name=pack, swing1LimitDegrees=45, swing2LimitDegrees=10, twistLimitDegrees=5}]
