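# Animation Blueprint for the KnightCatRacket backpack that animates its zipper, aims and offsets the pack, and toggles a material element.
# Cached base pose feeds RigidBody, AnimDynamics or Trail simulation, selected by bools, with gravity override on the first two.
# The pack bone then looks at the body's pelvis and takes additive offsets, including one for skydiving.
# Events set "Hide Element 01" when the tagged pickaxe is equipped outside skydiving.
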AnimBlueprint /BRCosmetics/Accessories/FORT_Backpacks/Backpack_F_MED_KnightCatRacket/Meshes/F_MED_KnightCatRacket_Pack_AnimBP
target_skeleton=&/F_MED_KnightCatRacket_Pack_Skeleton
graphs [
 AnimGraph {nodes [
   n1 = Anim Dynamics - Bone: din_zipper_base!(
    ComponentPose=Local To Component(LocalPose=Use cached pose 'BasePose'())
    Alpha=1
    GravityOverride:Vector=Get Gravity Override()
    ExternalForce="0,0,0"
   ) -> (Pose) {
    angularSpringConstant=4.5
    bAngularSpring=true
    bUseGravityOverride=true
    bUsePlanarLimit=false
    alphaBoolBlend {blendOption=Linear}
    boundBone {boneName=din_zipper_base}
    componentAppliedLinearAccClamp=<3000 3000 3000>
    componentLinearAccScale=<0.3 0.3 0.3>
    componentLinearVelScale=<0.3 0.3 0.3>
    gravityOverride {z=-980}
    physicsBodyDefinitions [{
      boundBone {boneName=din_zipper_base}
      boxExtents=<10 12 12>
      constraintSetup {angularLimitsMax=<2.5 2.5 30>, angularLimitsMin=<-2.5 -2.5 -45>, angularTarget=<0 0 0>}
      localJointOffset=<2 0 0>
     }]
   }
   n2 = Trail controller - Bone: din_zipper_parent(ComponentPose=Local To Component(LocalPose=Use cached pose 'BasePose'()), Alpha=0.5) {
    bUsePlanarLimit=true
    relaxationSpeedScale=0.8
    planarLimits [{drivingBone {boneName=din_zipper_base}, planeTransform {location=<0 0 0>, rotation=<rot -90 -45 0>, scale=<1 1 1>}}]
    rotationLimits [{limitMax=<30 30 30>, limitMin=<-30 -30 -30>}, {limitMax=<10 10 10>, limitMin=<-10 -10 -10>}]
    rotationOffsets [<0 0 0>, <-1 0 0>]
    trailBone {boneName=din_zipper_parent}
    trailRelaxationSpeed {editorCurveData {keys [<at 0 30 RCIM_Linear>,<at 1 30 RCIM_Linear>]}}
   }
   n3 = Blend Poses by bool(
    BlendPose_0=Component To Local(ComponentPose=@n1)
    BlendPose_1=Component To Local(ComponentPose=@n2)
    bActiveValue=Get bAnimDynamicsEnabled() -> (bAnimDynamicsEnabled:bool)
   ) {blendType=Linear}
   n4 = RigidBody!(
    ComponentPose=Local To Component(LocalPose=Use cached pose 'BasePose'())
    Alpha
    OverrideWorldGravity:Vector=Get Gravity Override()
    ExternalForce
   ) -> (Pose) {
    bClampLinearTranslationLimitToRefPose=true
    bOverrideWorldGravity=true
    overridePhysicsAsset=&/F_MED_KnightCatRacket_Pack_Physics
    alphaBoolBlend {blendOption=Linear}
    componentAppliedLinearAccClamp=<3000 3000 3000>
    componentLinearAccScale=<0.3 0.3 0.3>
    componentLinearVelScale=<0.3 0.3 0.3>
    overrideWorldGravity {z=-980}
   }
   n5 = Component To Local(ComponentPose=@n4)
   n6 = Get bRigidBodyEnabled()
   n7 = Look At - Bone: pack!(
    ComponentPose=Local To Component(LocalPose=Blend Poses by bool(BlendPose_0=@n5, BlendPose_1=@n3, bActiveValue=@n6) {blendType=Linear})
    Alpha=1
    LookAtLocation:Vector=Get PelvisLocation() -> (PelvisLocation:Vector)
   ) -> (Pose) {
    alphaBoolBlend {blendOption=Linear}
    boneToModify {boneName=pack}
    lookAt_Axis {axis {y=0, z=-1}}
    lookUp_Axis {axis {x=0, z=1}}
    lookAtTarget=~
   }
   n8 = "Transform (Modify) Bone - Bone: pack"!(ComponentPose=@n7, bAlphaBoolEnabled:bool, Translation="0,-1,0", Rotation="0,0,30", Scale) -> (Pose) {
    alphaInputType=Bool
    rotationMode=BMM_Additive
    rotationSpace=BCS_BoneSpace
    translationMode=BMM_Additive
    translationSpace=BCS_BoneSpace
    alphaBoolBlend {blendInTime=0.3, blendOption=ExpInOut, blendOutTime=0.3}
    alphaScaleBiasClamp {bInterpResult=true}
    boneToModify {boneName=pack}
    rotation {roll=30}
    translation {y=-1}
   }
   n9 = "Transform (Modify) Bone - Bone: pack"(ComponentPose=@n8, bAlphaBoolEnabled=Get bIsSkyDiving(), Translation="0,0,1", Rotation="0,0,8") {
    alphaInputType=Bool
    rotationMode=BMM_Additive
    rotationSpace=BCS_BoneSpace
    translationMode=BMM_Additive
    translationSpace=BCS_BoneSpace
    alphaBoolBlend {blendInTime=0.3, blendOption=ExpInOut, blendOutTime=0.3}
    alphaScaleBiasClamp {bInterpResult=true}
    boneToModify {boneName=pack}
    rotation {roll=8}
    translation {z=1}
   }
   Output Pose(Result=Component To Local(ComponentPose=@n9)) {layerGroup=None}
   n10 = Local To Component(LocalPose=Input Pose())
   Save cached pose 'BasePose'(Pose=Component To Local(ComponentPose="Transform (Modify) Bone - Bone: pack"!(ComponentPose=@n10, Alpha:float, Translation="0,0,0", Rotation="0,0,0", Scale) -> (Pose) {translationMode=BMM_Additive, alphaBoolBlend {blendOption=Linear}, boneToModify {boneName=pack}})) {cacheName=BasePose}
  ]}
 EventGraph {nodes [
   Event BlueprintUpdateAnimation()
   > n1 = Gravity Override State Function DEPRECATED(CharacterPartAnimInstance=Self-Reference(), GameTimeInSeconds=GetGameTimeInSeconds(), GravityOverrideParams=Get Gravity Override Params())
   > Set Gravity Override(Gravity Override=@n1)
   n2 = Get bodyMesh() -> (bodyMesh:Skeletal Mesh Component Object Reference)
   > Is Valid(InputObject=@n2)
   > Set PelvisLocation(PelvisLocation:Vector=GetSocketLocation!(self:Scene Component Object Reference=@n2, InSocketName:Name=pelvis) -> (ReturnValue:Vector)) -> (Output_Get:Vector)
   n3 = Event OnCharacterCustomizationLoaded()
   > n4 = Sequence()
   n5 = GetPartSkeletalMeshComponent()
   n6 = Get Dynamic Material Instance(execute=[@n14.then, @n4.then_0], SkeletalMeshComponent=@n5)
   > n7 = GetOrCreateDynamicOverlayMaterial(MeshComponent=@n5, MaterialName=None)
   > n8 = Sequence()
   n9 = PickaxeItemDefinitionHasAnyMetaTag(Tags=Make Array("[0]"=(TagName="Cosmetics.MetaTags.KnightCatRacket")), bRequiresEquipped=true)
   > n10 = Branch(Condition=@n9)
   > n11 = Branch(Condition=Get bIsSkyDiving())
   SetScalarParameterValue(execute=[@n15.then, @n11.then, @n12.else, @n10.else], self=[@n6, @n7], ParameterName=Hide Element 01, Value=0)
   SetScalarParameterValue(execute=[@n11.else, @n12.then], self=[@n6, @n7], ParameterName=Hide Element 01, Value=1)
   n12 = Branch(execute=[@n16.then, @n8.then_1], Condition=@n9)
   n13 = Cast To FortPlayerPawnAthena(execute=@n4.then_1, Object=TryGetPawnOwner())
   > Set bodyMesh(bodyMesh:Skeletal Mesh Component Object Reference=GetSkeletalMeshForPartType(self=@n13)) -> (Output_Get:Skeletal Mesh Component Object Reference)
   Set bodyMesh(execute=@n13.CastFailed, bodyMesh=@n3.CustomBaseMesh)
   n14 = Event OnWeaponChanged()
   n15 = Event OnBeginSkydiving()
   n16 = Event OnEndSkydiving()
   TryGetPawnOwner()
  ]}
]
variables [
 bodyMesh {type=SkeletalMeshComponent}
 Gravity Override {type=Vector}
 Gravity Override Params {type=GravityOverrideParamsStruct}
 PelvisLocation {type=Vector}
]
